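# Material function in the SideFX library category that outputs a pseudorandom float in the range 0 to 1 from a 2D seed.
# It uses the common sine-hash: dot product with magic constants, sine scaled by 1/2π, large multiply, then Frac.
# Unlike the plain hash, the constant vector is also multiplied by the UV's distance from the origin.
# An additional scalar seed offsets the UV, which defaults to texture coordinates, to vary the pattern.
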MaterialFunction /WildEstate_FactionChest/Assets/FX/MF_SimpleRandom
graph {
 properties {description=Generates simple pseudorandom float using a popular algorithm., libraryCategoriesText [SideFX]}
 nodes [
  n1 = FunctionInput(Preview=TextureCoordinate()) {
   bUsePreviewValueAsDefault=true
   description=A 2D seed to generate a pseudorandom float that is random on a 2D input domain.
   inputName="UV (2D Seed)"
   inputType=FunctionInput_Vector2
  }
  n2 = FunctionInput() {
   bUsePreviewValueAsDefault=true
   description="Changing this seed can generate different sets of results on the same 2D input domain. (Keep this value under 10000.)"
   inputName=Additional Seed
   inputType=FunctionInput_Scalar
   sortPriority=1
  }
  n3 = Add(A=@n1, B=@n2)
  n4 = Multiply(A=Constant2Vector() {g=78.233, r=12.9898}, B=Distance(A=@n1, B=Constant2Vector()))
  n5 = Sine(None=Multiply(A=DotProduct(A=@n3, B=@n4)) {constB=0.1592, desc="B = 1/2π"})
  n6 = Constant() {r=43758.5469}
  FunctionOutput(None=Frac(None=Multiply(A=@n5, B=@n6) {constB=43758.5469})) {description="Outputs a random float in the range of [0, 1). The value is random on a 2D domain.", outputName=Random Float, sortPriority=1}
 ]
}
