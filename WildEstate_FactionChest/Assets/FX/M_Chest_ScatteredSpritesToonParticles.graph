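# Additive unlit, two-sided Niagara sprite material for scattered toon glow particles in the faction chest FX.
# A radial sphere mask forms a tight, desaturated center and a softer edge glow combined with a contrasted star flare texture, all tinted by particle color.
# The sum feeds a shared emissive adjustments function.
# An Alpha reroute multiplies particle alpha by opacity, depth-faded on the deferred path; two SpriteColor parameters are unconnected.
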Material /WildEstate_FactionChest/Assets/FX/M_Chest_ScatteredSpritesToonParticles
graph {
 properties {
  bUsedWithNiagaraSprites=true
  blendMode=BLEND_Additive
  pixelDepthOffsetMode=PDOM_Legacy
  shadingModel=MSM_Unlit
  twoSided=true
  usedShadingModels=Unlit
 }
 nodes [
  n1 = Constant() {r=0.5}
  n2 = SphereMask(A=TextureCoordinate(), B=@n1, Radius=@n1) {hardnessPercent=0}
  n3 = Multiply(
   A=Power(Base=@n2, Exp=ScalarParameter() {defaultValue=25, parameterName=CentertPower}) {constExponent=25}
   B=ScalarParameter() {defaultValue=2, parameterName=CenterBrightness}
  ) {constB=2}
  n4 = ParticleColor()
  n5 = Desaturation(None=@n4.RGB, Fraction=ScalarParameter() {defaultValue=0.3, parameterName=CenterDesaturation})
  n6 = ScalarParameter() {defaultValue=4, parameterName=EdgePower}
  n7 = Add(
   A=Multiply(A=Power(Base=@n2, Exp=@n6) {constExponent=4}, B=ScalarParameter() {defaultValue=0.15, parameterName=EdgeBrightness}) {constB=0.15}
   B=MaterialFunctionCall(In=TextureSampleParameter2D() {parameterName=FlareMask, texture=&/star_glow}.R, Contrast=Constant() {r=5}) {materialFunction=&CheapContrast}.Result
  )
  MaterialFunctionCall(Emissive=Add(A=Multiply(A=@n3, B=@n5), B=Multiply(A=@n4.RGB, B=@n7))) {materialFunction=&MF_FX_EmissiveAdjustments}
  n8 = ScalarParameter() {defaultValue=1, parameterName=Material Opacity}
  n9 = DepthFade(Opacity=@n8, FadeDistance=ScalarParameter() {parameterName=DepthFade})
  NamedRerouteDeclaration(None=Multiply(A=ShadingPathSwitch(Default=@n8, Deferred=@n9), B=@n4.A)) {name=Alpha, nodeColor=<rgba 1 0.8672 0 1>}
  VectorParameter() {parameterName=SpriteColor_High, defaultValue=<rgba 4 3 2.25 1>}
  VectorParameter() {parameterName=SpriteColor_Low, defaultValue=<rgba 4 1.625 0 1>}
 ]
}
