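# Default Lit material for the Wild Estate faction chest, apparently a temporary placeholder judging by its Temp name.
# Its graph holds a single texture sample of the WildEstate_FactionChest_Temp texture, with no connection to a material output shown.
# It is flagged for instanced static meshes, and automatic usage setting in the editor is turned off.
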Material /WildEstate_FactionChest/Assets/WildEstate_FactionChest_Temp_Mat
graph {
 properties {bAutomaticallySetUsageInEditor=false, bUsedWithInstancedStaticMeshes=true, usedShadingModels=Default Lit}
 nodes [TextureSample() {texture=&/WildEstate_FactionChest_Temp}]
}
